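# Unlit, translucent, two-sided Control Rig material, apparently for trajectories, drawn with depth testing disabled.
# Compares the V texture coordinate to the Time parameter, blending green toward a white highlight that trails off just below Time.
# Vertex color red looks up T_Traj, whose RGB tints the gradient.
# A fixed fake light direction dotted with the world normal adds shading without scene lighting.
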Material /ControlRig/M_Traj
graph {
 properties {
  bDisableDepthTest=true
  blendMode=BLEND_Translucent
  pixelDepthOffsetMode=PDOM_Legacy
  shadingModel=MSM_Unlit
  twoSided=true
  usedShadingModels=Unlit
 }
 nodes [
  n1 = VertexColor()
  n2 = Constant3Vector() {constant=<rgba 0 0.506 0.0169 1>}
  n3 = Constant3Vector() {constant=<rgba 1 1 1 1>}
  n4 = MaterialFunctionCall(Float2=TextureCoordinate()) {materialFunction=&BreakOutFloat2Components}
  n5 = ScalarParameter() {parameterName=Time}
  n6 = Constant() {r=0.3}
  n7 = Clamp(None=Abs(None=Subtract(A=@n4.G, B=@n5)), Max=@n6)
  n8 = If(A=@n4.G, B=@n5, "A > B"=Constant(), "A < B"=Subtract(B=Divide(A=@n7, B=@n6)))
  n9 = Multiply(A=Constant3Vector() {desc=Fake light direction, constant=<rgba 1 1.5 -3 0>}) {constB=-1}
  n10 = ConstantBiasScale(None=DotProduct(A=Normalize(VectorInput=@n9), B=VertexNormalWS()))
  Multiply(
   A=TextureSample(UVs=MaterialFunctionCall(X=@n1.R, Y=@n1.R) {materialFunction=&MakeFloat2}.Result) {texture=&/T_Traj}.RGB
   B=Multiply(A=LinearInterpolate(A=@n2, B=@n3, Alpha=@n8), B=Add(A=@n10) {constB=0.1})
  )
 ]
}
